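# Base widget for cycling videos, built on FortVideoCycleBase, that fades its video in and out.
# An overlay holds a FortVideoWidget that starts transparent, with activation focus disabled and empty skip actions.
# PlayVideoIntro plays the short render-opacity animation forward and PlayVideoOutro reverses it.
# On Construct, a SnapToEndOfAnimation helper plays that animation in reverse from its end time, intended to snap straight to hidden.
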WidgetBlueprint /FortniteCoreUI/UI/Foundation/Movie/WBP_FortVideoCycle_Base
animations [FadeInRenderOpacity {
  binding_count=1
  length_seconds=0.4
  object=&self:FadeInRenderOpacity
  sequence {bindings [FortVideoWidget {
     id=#1
     possessed_class=WBP_FortVideo_Base_C
     tracks [{
       type=MovieSceneFloatTrack
       display_name=Render Opacity
       sections [{type=MovieSceneFloatSection, end_frame=8, start_frame=0, channels [None {type=MovieSceneScriptingFloatChannel, keys [[0,0],[8,1]]}]}]
      }]
    }]}
 }]
graphs [
 EventGraph {nodes [
   Event Construct()
   > SnapToEndOfAnimation!(
    execute
    self
    InAnimation:Widget Animation Object Reference=Get FadeInRenderOpacity() -> (FadeInRenderOpacity:Widget Animation Object Reference)
    PlayMode:EUMGSequencePlayMode Enum=Reverse
   ) -> (then)
   Event PlayVideoIntro() {is=title K2Node_Event}
   > PlayAnimationForward(InAnimation=Get FadeInRenderOpacity())
   Event PlayVideoOutro() {is=title K2Node_Event}
   > PlayAnimationReverse(InAnimation=Get FadeInRenderOpacity())
   Event PreConstruct()
  ]}
 SnapToEndOfAnimation {nodes [
   n1 = SnapToEndOfAnimation!() -> (then, InAnimation:Widget Animation Object Reference, PlayMode:EUMGSequencePlayMode Enum=Forward) {is=title K2Node_FunctionEntry}
   > PlayAnimation(InAnimation=@n1, StartAtTime=GetEndTime(self=@n1), PlayMode=@n1.PlayMode)
  ]}
]
variables []
widgets {
 info {inheritedWidgetCount=1, parentClass=/Script/FortniteUI.FortVideoCycleBase, rootWidgetClass=&Overlay, widgetCount=2}
 widgets [
  MainOverlay {widgetClassPath=&Overlay, widget=MainOverlay}
  FortVideoWidget {
   widgetClassPath=/Game/UI/Foundation/Movie/WBP_FortVideo_Base
   bInherited=true
   bIsVariable=true
   parent=MainOverlay
   widget=FortVideoWidget
   properties {
    bSupportsActivationFocus=false
    renderOpacity=0
    holdToSkipAction {dataTable=null, rowName=None}
    pressToSkipAction {dataTable=null, rowName=None}
   }
   slot {class=OverlaySlot}
  }
 ]
}
