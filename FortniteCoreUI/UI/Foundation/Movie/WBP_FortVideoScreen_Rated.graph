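# Full-screen rated video screen, derived from FortRatedVideoScreen, that plays a video cycle with Skip and optional call-to-action buttons.
# An interstitial viewmodel supplies the video and CTA button data through SetVMVideoInfo and SetVMButton.
# Skip stops playback, while CTA runs its MOTD action and may close the screen, suppressing the next interstitial.
# On video end it sends analytics, conditionally marks the interstitial shown, dismisses it and deactivates.
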WidgetBlueprint /FortniteCoreUI/UI/Foundation/Movie/WBP_FortVideoScreen_Rated
graphs [
 BP_GetDesiredFocusTarget {nodes [
   Get Desired Focus Target()
   > n1 = Is Interstitial Video!(Exec:Exec) -> (True:Exec, False:Exec) {is=title K2Node_MacroInstance}
   > n2 = Branch(Condition=HasValidButtonData!(self:Video Interstitial Viewmodel Object Reference=Get VideoInterstitialVM() -> (VideoInterstitialVM:Video Interstitial Viewmodel Object Reference)) -> (ReturnValue:bool))
   > Return Node!(execute, ReturnValue:Widget Object Reference=Get Button_CTA() -> (Button_CTA:WBP UIKit Button Loud Object Reference))
   Return Node(execute=[@n2.else, @n1.False], ReturnValue=Get Button_Skip() -> (Button_Skip:WBP UIKit Button Regular Object Reference))
  ]}
 EventGraph {nodes [
   "On Clicked (Button_Skip)"!() -> (OutputDelegate, then, Button:Common Button Base Object Reference) {is=title K2Node_ComponentBoundEvent}
   > Fire Event Element Interacted(Element=Self-Reference())
   > n1 = Is Interstitial Video()
   > n2 = OnButtonPressed!(
    execute
    self:Fort Interstitial VM Object Reference=Get VideoInterstitialVM()
    ButtonType:EInterstitialButtonType Enum=Close
    Style:EMotdButtonStyle Enum=Secondary
    Action:Motd Action Base Object Reference
   ) -> (then)
   StopVideo(execute=[@n2.then, @n1.False], self:Fort Rated Video Screen Object Reference)
   "On Clicked (Button_CTA)"!() -> (OutputDelegate, then, Button:Common Button Base Object Reference) {is=title K2Node_ComponentBoundEvent}
   > Fire Event Element Interacted(Element=Self-Reference())
   > Is Interstitial Video()
   > n3 = TryExecuteButtonAction!(execute, self) -> (then, ShouldCloseWidget:bool)
   > Branch(Condition=@n3)
   > Set ShouldShowNext(ShouldShowNext:bool=false) -> (Output_Get:bool)
   > StopVideo()
   n4 = InterstitialVideoEndedAnalytics!() -> (OutputDelegate, then, VideoEndSnapshop:Fort Video Cycle Debug Snapshot Structure) {is=title K2Node_CustomEvent}
   n5 = Break Fort Video Cycle Debug Snapshot!(FortVideoCycleDebugSnapshot:"Fort Video Cycle Debug Snapshot Structure (by ref)"=@n4.VideoEndSnapshop) -> (UID:String, CurrentPlaytimeSeconds:float=~, Description:String) {is=title K2Node_BreakStruct}
   > n6 = Add Event Info String!(
    execute
    EventInfo:"Interstitial Event Info Structure (by ref)"=Make Interstitial Event Info!(Name:String=VideoEndEvent) -> (InterstitialEventInfo:Interstitial Event Info Structure) {is=title K2Node_MakeStruct}
    Key:String=VideoUID
    Value:String=@n5
   ) -> (then, ReturnValue:Interstitial Event Info Structure)
   > n7 = Add Event Info Float!(execute, EventInfo:"Interstitial Event Info Structure (by ref)"=@n6, Key:String=Playtime, Value:float=@n5.CurrentPlaytimeSeconds) -> (then, ReturnValue:Interstitial Event Info Structure)
   > n8 = Add Event Info String(EventInfo=@n7, Key=EndResult, Value=@n5.Description)
   > TriggerAnalyticsEvent!(execute, self:Fort Interstitial VM Object Reference=Get VideoInterstitialVM(), EventInfo:"Interstitial Event Info Structure (by ref)"=@n8) -> (then)
   n9 = Event CallVideoEnded!() -> (
    OutputDelegate
    then
    EndReason:"EFortVideoScreenClosedReason Enum (by ref)"=EndReached
    EndSnapshot:"Fort Video Cycle Debug Snapshot Structure (by ref)"
   ) {is=title K2Node_Event}
   > n10 = Is Valid(InputObject=Get VideoInterstitialVM())
   > InterstitialVideoEndedAnalytics!(execute, self, VideoEndSnapshop:Fort Video Cycle Debug Snapshot Structure=@n9.EndSnapshot) -> (then) {is=title K2Node_CallFunction}
   > n11 = Should Mark Shown!(Exec:Exec, EndReason:EFortVideoScreenClosedReason Enum=@n9.EndReason) -> (True:Exec, False:Exec) {is=title K2Node_MacroInstance}
   n12 = Get VideoInterstitialVM()
   > n13 = InterstitialWasShown(self:Fort Interstitial VM Object Reference=@n12)
   DismissInterstitial!(
    execute=[@n13.then, @n11.False]
    self:Fort Interstitial VM Object Reference=@n12
    bShouldShowNext:bool=Get ShouldShowNext() -> (ShouldShowNext:bool)
   ) -> (then)
   > n14 = Set VideoInterstitialVM(VideoInterstitialVM:Video Interstitial Viewmodel Object Reference) -> (Output_Get:Video Interstitial Viewmodel Object Reference)
   DeactivateWidget(execute=[@n14.then, @n10.Is Not Valid])
  ]}
 GetButtons {nodes [
   GetButtons() {is=title K2Node_FunctionEntry}
   n1 = Get RetrievedButtons() -> (RetrievedButtons:Array of Fort CTAButton Object References)
   > Add(TargetArray:Array of Fort CTAButton Object References=@n1, NewItem:"Fort CTAButton Object Reference (by ref)"=Get Button_Skip()) {is=title K2Node_CallArrayFunction}
   > n2 = Is Interstitial Video()
   > n3 = Branch(Condition=HasValidButtonData(self=Get VideoInterstitialVM()))
   n4 = Get RetrievedButtons()
   > Add(TargetArray=@n4, NewItem=Get Button_CTA())
   > Return Node(ReturnValue:Array of Fort CTAButton Object References=@n4)
   Return Node(execute=[@n3.else, @n2.False], ReturnValue=@n1)
  ]}
 GetPrimaryContentSetup {nodes [
   GetPrimaryContentSetup() {is=title K2Node_FunctionEntry}
   > Return Node(ReturnValue:Primary Content Setup Structure=Make Primary Content Setup!(
     bShowBottomBar:bool=false
     bShowChatWidget:bool=false
     bEnableNewTextChat:bool=False
     bShowTopBar:bool=false
     bTabSupportsInterface:bool=False
    ) -> (PrimaryContentSetup:Primary Content Setup Structure) {is=title K2Node_MakeStruct})
  ]}
 IsInterstitialVideo {nodes [
   Inputs!() -> (Exec:Exec)
   n1 = Get VideoInterstitialVM()
   > n2 = Is Valid(InputObject=@n1)
   > n3 = Branch(Condition=HasValidData!(self:Fort Interstitial VM Object Reference=@n1) -> (ReturnValue:bool))
   > Outputs!(True:Exec, False:Exec=[@n3.else, @n2.Is Not Valid])
  ]}
 SetVMButton {nodes [
   n1 = SetVMButton!() -> (then, HasValidButtonData:bool=~) {is=title K2Node_FunctionEntry}
   > n2 = Branch(Condition=@n1)
   n3 = Break Motd Button!(MotdButton:"Motd Button Structure (by ref)"=Get DynamicButton!(self:Video Interstitial Viewmodel Object Reference=Get VideoInterstitialVM()) -> (DynamicButton:Motd Button Structure)) -> (Text:Text, Style:EMotdButtonStyle Enum, Action:Motd Action Base Object Reference) {is=title K2Node_BreakStruct}
   > SetText!(execute, self:Fort CTAButton Object Reference=Get Button_CTA(), ButtonText:"Text (by ref)"=@n3) -> (then)
   > Set ButtonAction(ButtonAction:Motd Action Base Object Reference=@n3.Action) -> (Output_Get:Motd Action Base Object Reference)
   SetVisibility(execute=@n2.else, self=Get Button_CTA(), InVisibility=Collapsed)
  ]}
 SetVMVideoInfo {nodes [
   n1 = SetVMVideoInfo!() -> (then, InVideoInfo:Fort Video Cycle Info Structure) {is=title K2Node_FunctionEntry}
   n2 = Break Fort Video Cycle Info!(FortVideoCycleInfo:"Fort Video Cycle Info Structure (by ref)"=@n1) -> (UID:String, Title:String) {is=title K2Node_BreakStruct}
   n3 = IsEmpty(InString=@n2)
   n4 = IsEmpty(InString=@n2.Title)
   > Branch(Condition=NOT Boolean(A=OR Boolean(A=@n3, B=@n4)))
   > Set ShouldShowNext(ShouldShowNext=true)
   > PlayVideo!(
    execute
    self:Fort Rated Video Screen Object Reference
    VideoCycleInfo:"Fort Video Cycle Info Structure (by ref)"=@n1
    ButtonsEnableTime:float=0
    bLooped:bool=false
   ) -> (then)
  ]}
 ShouldMarkShown {nodes [
   n1 = Inputs!() -> (Exec, EndReason:EFortVideoScreenClosedReason Enum=EndReached)
   n2 = Get VideoInterstitialVM()
   n3 = IsValid(Object=@n2)
   n4 = NOT Boolean(A=Get bShouldMarkAsShownOnPush!(self:Fort Interstitial VM Object Reference=@n2) -> (bShouldMarkAsShownOnPush:bool))
   > n5 = Branch(Condition=AND Boolean(
     A=AND Boolean(A=@n3, B=@n4)
     B="Not Equal (Enum)"!(A:EFortVideoScreenClosedReason Enum=@n1, B:EFortVideoScreenClosedReason Enum=Error) -> (ReturnValue:bool)
    ))
   > Outputs(False=@n5.else)
  ]}
 TryExecuteButtonAction {nodes [
   TryExecuteButtonAction() {is=title K2Node_FunctionEntry}
   > n1 = Branch(Condition=HasValidButtonData(self=Get VideoInterstitialVM()))
   > n2 = Cast To FortPlayerController(Object=GetOwningPlayer())
   > OnButtonPressed(
    self=Get VideoInterstitialVM()
    ButtonType=ActionButton
    Style=Main
    Action=Get ButtonAction() -> (ButtonAction:Motd Action Base Object Reference)
   )
   n3 = Get ButtonAction()
   > CallToAction!(execute, self:Motd Action Base Object Reference=@n3, PlayerController:Fort Player Controller Object Reference=@n2) -> (then)
   > Return Node!(execute, ShouldCloseWidget:bool=ShouldCloseWidget!(self:Motd Action Base Object Reference=@n3) -> (ReturnValue:bool))
   Return Node(execute=[@n2.CastFailed, @n1.else], ShouldCloseWidget=false)
  ]}
 bEnableButtonBinds {nodes [
   bEnableButtonBinds() {is=title K2Node_FunctionEntry}
   > n1 = Branch(Condition=GetConfigDataBoolWithDefault!(EntryName:String=bEnableVIdeoButtonBinds, bDefaultValue:bool=false) -> (ReturnValue:bool))
   > SetTriggeringInputAction!(
    execute
    self:Common Button Base Object Reference=Get Button_Skip()
    InputActionRow:"Data Table Row Handle Structure (by ref)"=Get InputActionBack() -> (InputActionBack:Data Table Row Handle Structure)
   ) -> (then)
   > SetTriggeringInputAction(self=Get Button_CTA(), InputActionRow=Get InputActionForward() -> (InputActionForward:Data Table Row Handle Structure))
   n2 = Get InputActionNone() -> (InputActionNone:Data Table Row Handle Structure)
   SetTriggeringInputAction(execute=@n1.else, self=Get Button_Skip(), InputActionRow=@n2)
   > SetTriggeringInputAction(self=Get Button_CTA(), InputActionRow=@n2)
  ]}
]
variables [
 ButtonAction {type=MotdActionBase, default=null}
 ShouldShowNext {type=boolean, default=true}
]
widgets {
 info {inheritedWidgetCount=2, parentClass=/Script/FortniteUI.FortRatedVideoScreen, rootWidgetClass=&Border, widgetCount=8}
 widgets [
  SafezonePadding {
   widgetClassPath=&Border
   widget=SafezonePadding
   properties {visibility=SelfHitTestInvisible, brushColor=<rgba 0 0 0 1>, padding=<ltrb 0 0 0 0>}
  }
  MainOverlay {
   widgetClassPath=&Overlay
   parent=SafezonePadding
   widget=MainOverlay
   slot {class=BorderSlot, properties {padding=<ltrb 0 0 0 0>}}
  }
  FortVideoCycleWidget {
   widgetClassPath=&/WBP_FortVideoCycle_Base
   bInherited=true
   bIsVariable=true
   parent=MainOverlay
   widget=FortVideoCycleWidget
   slot {class=OverlaySlot, properties {horizontalAlignment=HAlign_Center, verticalAlignment=VAlign_Center}}
  }
  SqueegeeSlot_Bumper {
   widgetClassPath=/Game/UI/Plugin/Squeegee/SqueegeeInjectionSlot
   bInherited=true
   bIsVariable=true
   parent=MainOverlay
   widget=SqueegeeSlot_Bumper
   properties {pluginId=Squeegee, pluginWidgetId=Video Bumper, renderOpacity=0}
   slot {class=OverlaySlot, properties {horizontalAlignment=HAlign_Center, verticalAlignment=VAlign_Center}}
  }
  SafeZone_Buttons {
   widgetClassPath=&SafeZone
   parent=MainOverlay
   widget=SafeZone_Buttons
   properties {padLeft=false, padTop=false}
   slot {class=OverlaySlot, properties {horizontalAlignment=HAlign_Fill, verticalAlignment=VAlign_Bottom, padding=<ltrb 0 0 32 32>}}
  }
  WrapBox_0 {
   widgetClassPath=&WrapBox
   parent=SafeZone_Buttons
   widget=WrapBox_0
   properties {bOverride_Cursor=true, horizontalAlignment=HAlign_Right, innerSlotPadding=<20 0>}
   slot {class=SafeZoneSlot, properties {vAlign=VAlign_Top}}
  }
  Button_CTA {
   widgetClassPath=/FortUILibrary/UIKit/Buttons/WBP_UIKit_Button_Loud
   bIsVariable=true
   parent=WrapBox_0
   widget=Button_CTA
   properties {bDisplayInputActionWhenNotInteractable=false, bShouldUseFallbackDefaultInputAction=false, visibility=Collapsed}
   slot {class=WrapBoxSlot}
  }
  Button_Skip {
   widgetClassPath=&WBP_UIKit_Button_Regular
   bIsVariable=true
   parent=WrapBox_0
   widget=Button_Skip
   properties {bDisplayInputActionWhenNotInteractable=false, bShouldUseFallbackDefaultInputAction=false, text=Skip, visibility=Collapsed}
   slot {class=WrapBoxSlot}
  }
 ]
}
